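; Looping gameplay cue that drives riding audio while a player rides an AI creature in LabradorRiding.
; It plays attached mount and dismount sounds plus a team-affiliation-selected loop, unless the player's gameplay tags block them, caching that check for dismount.
; A repeating timer normalizes the attached creature's velocity against run speed and feeds an interpolated "Speed" parameter to the loop.
; Removal clears the timer and fades the loop out.
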(event EventBeginPlay)

(event Collision|EventActorBeginOverlap
  (params (OtherActor :type Actor%20Object%20Reference)))

(event EventTick
  (params (DeltaSeconds :type Float%20%28single-precision%29)))

(event EventOnLoopingStart(EitherVFX-preferthisversion)
  (params (MyTarget :type Actor%20Object%20Reference) (Parameters :type Gameplay%20Cue%20Parameters%20Structure%20%28by%20ref%29) (ParticleComponents :type Array%20of%20FXSystem%20Component%20Object%20References) (AudioComponents :type Array%20of%20Audio%20Component%20Object%20References))
  (CallFunction|PlayerRidingGCNLStarted))

(event EventOnRemoval(EitherVFX-preferthisversion)
  (params (MyTarget :type Actor%20Object%20Reference) (Parameters :type Gameplay%20Cue%20Parameters%20Structure%20%28by%20ref%29) (ParticleComponents :type Array%20of%20FXSystem%20Component%20Object%20References) (AudioComponents :type Array%20of%20Audio%20Component%20Object%20References) (BurstCameraShakeInstance :type Camera%20Shake%20Base%20Object%20Reference) (BurstDecalInstance :type Decal%20Actor%20Object%20Reference))
  (CallFunction|PlayerRidingDCNLEnded)
  (Utilities|Time|ClearandInvalidateTimerbyHandle 0 (Variables|Default|GetVelocityTimerHandle))
  (Audio|Components|Audio|FadeOut (Variables|Default|GetRidingSoundLoopComp) 0.1 0.0 "Linear" :!class /Script/Engine.AudioComponent)
  (Variables|Default|SetSpeed 0.0))

(event Custom|PlayerRidingGCNLStarted
  ;[1 0.586 0 0.2] Apply Sound Cues for entering/exiting riding if we aren't explicitly blocking them (some objects may not want these and will block them via tags)
  (bind _player (Variables|Default|GetPlayer)
    (:then
      (bind _returnvalue (GameplayTags|HasAnyMatchingGameplayTags(Message) _player (Variables|Default|GetBlockSitAudioTags)))
      (bind _output_get (Variables|Default|SetHasAudio (not _returnvalue))) ; Caching here as we want to persist this for the duration of this \nriding session (in case applying GE is removed before this GCN)
      (if _output_get
        (bind _burstoncue (Variables|Default|GetBurstOnCue)
          (:then
            (Audio|SpawnSoundAttached _burstoncue (Variables|/Script/Engine.Actor|GetRootComponent _player)))
          (:else))))
    (:else)))

(event Custom|PlayerRidingDCNLEnded
  ;[1 0.586 0 0.2] Apply Sound Cues for entering/exiting riding if we aren't explicitly blocking them (some objects may not want these and will block them via tags)
  (bind _player (Variables|Default|GetPlayer)
    (:then
      (if (Variables|Default|GetHasAudio)
        (bind _burstoffcue (Variables|Default|GetBurstOffCue)
          (:then
            (Audio|SpawnSoundAttached _burstoffcue (Variables|/Script/Engine.Actor|GetRootComponent _player)))
          (:else))))
    (:else)))

(event Custom|StartLoopingAudio
  (bind _teamaffiliation (Variables|Default|GetTeamAffiliation))
  (bind _enemy_ridingsoundloop (Variables|Default|GetEnemy_RidingSoundLoop))
  (bind _ridingsoundloop (Variables|Default|GetRidingSoundLoop))
  (bind _returnvalue (Utilities|Select :Index _teamaffiliation :Friendly _ridingsoundloop :Neutral _ridingsoundloop :Hostile _enemy_ridingsoundloop))
  ;[1 1 1] Audio - Looping audio
  (bind _player (Variables|Default|GetPlayer)
    (:then
      (bind _returnvalue_2 (GameplayTags|HasAnyMatchingGameplayTags(Message) _player (Variables|Default|GetBlockSitAudioTags)))
      (if (not _returnvalue_2)
        (bind _ridingsoundloopcomp (Variables|Default|GetRidingSoundLoopComp)
          (:then
            (Audio|Components|Audio|FadeOut _ridingsoundloopcomp 0.1 0.0 "Linear" :!class /Script/Engine.AudioComponent))
          (:else
            _))
        (bind _player_2 (Variables|Default|GetPlayer)
          (:then
            (Variables|Default|SetTeamAffiliation (Team|GetTeamAffiliationforActortoLocalPlayer _player_2))
            (bind _rootcomponent (Variables|/Script/Engine.Actor|GetRootComponent _player_2)
              (:then
                (Utilities|IsValid _returnvalue
                  (:Is%20Valid
                    (bind _returnvalue_3 (Audio|SpawnSoundAttached :Sound _returnvalue :AttachToComponent _rootcomponent :AttachPointName "None" :LocationType "KeepRelativeOffset" :bStopWhenAttachedToDestroyed true :VolumeMultiplier 1.0 :PitchMultiplier 1.0 :StartTime 0.0 :bAutoDestroy true))
                    (Variables|Default|SetRidingSoundLoopComp _returnvalue_3)
                    (bind _returnvalue_4 (Utilities|Time|SetTimerbyFunctionName 0 "CalculateVelocityParam" 0.1 true))
                    (Variables|Default|SetVelocityTimerHandle _returnvalue_4))))
              (:else)))
          (:else))))
    (:else)))

(event Custom|SetUpRidableAudioEvents
  ;[1 1 1] Initialization - Audio
  (Utilities|FlowControl|Delay 0)
  (bind _player (Variables|Default|GetPlayer)
    (:then
      (Utilities|IsValid (Actor|GetAttachParentActor _player)
        (:Is%20Valid
          (bind _asfort_aipawn (Utilities|Casting|CastToFortAIPawn (Actor|GetAttachParentActor _player)))
          (Variables|Default|SetCreatureMovementComponent (Pawn|GetMovementComponent _asfort_aipawn)))))
    (:else)))

(event Custom|OnJumpApex
  ;[1 0.2392 0.2392] Audio - Jump affecting pitch of saddler loop sound (DEPRECATED)
  (Variables|Default|SetJumpParamTarget 0.0))

(fn CalculateVelocityParam
  (Variables|Default|SetMaxSpeed (Variables|Default|GetMaxNormalRunSpeed))
  (bind _creaturemovementcomponent (Variables|Default|GetCreatureMovementComponent)
    (:then
      (bind _output_get (Variables|Default|SetCurrentVelocity (Math|Vector|VectorLength (Variables|/Script/Engine.MovementComponent|GetVelocity _creaturemovementcomponent))))
      (Variables|Default|SetMaxSpeedClamped (Math|Float|MapRangeClamped _output_get 0.0 (Variables|Default|GetMaxSpeed) 0.0 1.0))
      (bind _ridingsoundloopcomp (Variables|Default|GetRidingSoundLoopComp)
        (:then
          (Audio|Parameter|SetFloatParameter _ridingsoundloopcomp "Speed" (Variables|Default|GetSpeed) :!class /Script/AudioExtensions.AudioParameterControllerInterface)
          (Variables|Default|SetSpeed (Math|Interpolation|FInterpTo (Variables|Default|GetSpeed) (Variables|Default|GetMaxSpeedClamped) (Utilities|Time|GetWorldDeltaSeconds 0) 1e+01)))
        (:else)))
    (:else)))

(fn ConstructionScript)

(fn OnBecomeRelevant
  (params (MyTarget :type Actor%20Object%20Reference) (Parameters :type Gameplay%20Cue%20Parameters%20Structure%20%28by%20ref%29))
  (Variables|Default|SetPlayer MyTarget)
  (CallFunction|SetUpRidableAudioEvents)
  (CallFunction|StartLoopingAudio)
  (CallFunction|PlayerRidingGCNLStarted)
  (return false))
