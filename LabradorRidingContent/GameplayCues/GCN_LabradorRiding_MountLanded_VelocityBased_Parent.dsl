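; Parent GameplayCue notify for a riding mount landing that plays force feedback and camera shake scaled by landing velocity.
; OnExecute reads the absolute raw magnitude as velocity, then selects damage or no-damage effect arrays and velocity range against a damage threshold.
; A clamped range map turns velocity into a rounded index for each array, and the chosen effects play at the cue location.
; Optional flags print the velocity and selected asset names for debugging.
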(fn OnExecute
  (params (MyTarget :type Actor%20Object%20Reference) (Parameters :type Gameplay%20Cue%20Parameters%20Structure%20%28by%20ref%29))
  (bind (_normalizedmagnitude _rawmagnitude _effectcontext _matchedtagname _originaltag _aggregatedsourcetags _aggregatedtargettags _location) (Ability|GameplayCue|BreakGameplayCueParameters Parameters))
  (bind _selectedforcefeedbackindex (LocalVariables|GetSelectedForceFeedbackIndex))
  (bind _selectedforcefeedbackarray (LocalVariables|GetSelectedForceFeedbackArray))
  (bind _output (Utilities|Array|Get(acopy) _selectedforcefeedbackarray _selectedforcefeedbackindex))
  (bind _selectedcamerashakeindex (LocalVariables|GetSelectedCameraShakeIndex))
  (bind _selectedcamerashakearray (LocalVariables|GetSelectedCameraShakeArray))
  (bind _output_2 (Utilities|Array|Get(acopy) _selectedcamerashakearray _selectedcamerashakeindex))
  (bind _selectedcamerashakearray_2 (LocalVariables|GetSelectedCameraShakeArray))
  (bind _selectedforcefeedbackarray_2 (LocalVariables|GetSelectedForceFeedbackArray))
  (LocalVariables|SetLocation _location)
  (bind _output_get (LocalVariables|SetLandingVelocity (Math|Float|Absolute(Float) _rawmagnitude)))
  (if (Variables|Default|GetPrintVelocityDebug)
    (Development|PrintString 0 (Utilities|String|Append "Mount fall velocity upon landing = " (Utilities|String|ToString(Float) _output_get) "
"))
    (else
      _))
  (Utilities|FlowControl|Sequence
    (:then_0
      ;[0.6302 1 0.7843] Initialize values
      (if (>= (LocalVariables|GetLandingVelocity) (Variables|Default|GetMinVelocityToGetDamaged))
        (LocalVariables|SetSelectedForceFeedbackArray (Variables|Default|GetForceFeedbacks_Damage))
        (LocalVariables|SetInRangeA (Variables|Default|GetMinVelocityToGetDamaged))
        (LocalVariables|SetInRangeB (Variables|Default|GetMaxVelocity))
        (LocalVariables|SetSelectedCameraShakeArray (Variables|Default|GetCameraShakes_Damage))
        (else
          (LocalVariables|SetSelectedForceFeedbackArray (Variables|Default|GetForceFeedbacks_NoDamage))
          (LocalVariables|SetInRangeA (Variables|Default|GetMinVelocity))
          (LocalVariables|SetInRangeB (Variables|Default|GetMinVelocityToGetDamaged))
          (LocalVariables|SetSelectedCameraShakeArray (Variables|Default|GetCameraShakes_NoDamage)))))
    (:then_1
      ;[0.4077 0.7645 1] Pick Force Feedback Index
      (if (Utilities|IsArrayNonEmpty _selectedforcefeedbackarray_2)
        (LocalVariables|SetSelectedForceFeedbackIndex (Math|Float|Round (Math|Float|MapRangeClamped (LocalVariables|GetLandingVelocity) (LocalVariables|GetInRangeA) (LocalVariables|GetInRangeB) 0.0 (Math|Conversions|ToFloat(Integer) (Utilities|Array|LastIndex _selectedforcefeedbackarray_2)))))
        ;[0.4077 0.7645 1] Play Force Feedback
        (Utilities|IsValid _output
          (:Is%20Valid
            (Fortnite|PlayLocalForceFeedbackatLocation 0 _output (LocalVariables|GetLocation) 3e+02)
            ;[0.4077 0.7645 1] Debug Force Feedback
            (if (Variables|Default|GetPrintForceFeedbackNameDebug)
              (Development|PrintString 0 (Utilities|GetDisplayName (Utilities|Array|Get(acopy) (LocalVariables|GetSelectedForceFeedbackArray) (LocalVariables|GetSelectedForceFeedbackIndex)))))))))
    (:then_2
      ;[0.5687 0.4189 1] Pick Camera Shake Index
      (if (Utilities|IsArrayNonEmpty _selectedcamerashakearray_2)
        (LocalVariables|SetSelectedCameraShakeIndex (Math|Float|Round (Math|Float|MapRangeClamped (LocalVariables|GetLandingVelocity) (LocalVariables|GetInRangeA) (LocalVariables|GetInRangeB) 0.0 (Math|Conversions|ToFloat(Integer) (Utilities|Array|LastIndex _selectedcamerashakearray_2)))))
        ;[0.5687 0.4189 1] Play Camera Shake
        (if (Utilities|IsValidClass _output_2)
          (Camera|PlayWorldCameraShake 0 _output_2 (LocalVariables|GetLocation) 0.0 5e+02 0.0)
          ;[0.5687 0.4189 1] Debug Camera Shake
          (if (Variables|Default|GetPrintCameraShakeNameDebug)
            (Development|PrintString 0 (Utilities|GetClassDisplayName (Utilities|Array|Get(acopy) (LocalVariables|GetSelectedCameraShakeArray) (LocalVariables|GetSelectedCameraShakeIndex))))))))
    (:then_3
      (return true))))
